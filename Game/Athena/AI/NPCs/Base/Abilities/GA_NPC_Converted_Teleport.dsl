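; Gameplay ability that teleports a converted NPC's player pawn to a location read from its AI Blackboard.
; After commit, it validates the Blackboard vector, stores it as the destination and records the pawn's starting location.
; GetTeleportDestination projects that point onto navigation data and raises it by capsule half-height, falling back to the raw point.
; A successful teleport fires start and end GameplayCues at the old and new locations.
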(event Ability|EventActivateAbility
  (bind _teleportlocationbbkeyname (Variables|Default|GetTeleportLocationBBKeyName))
  (Ability|CommitAbility)
  (bind _asfort_player_pawn (Utilities|Casting|CastToFortPlayerPawn (Ability|GetAvatarActorfromActorInfo))
    (:then
      (if (AI|IsValidAILocation (AI|Components|Blackboard|GetValueasVector (AI|GetBlackboard _asfort_player_pawn) _teleportlocationbbkeyname))
        (Variables|Default|SetTeleportDestination (AI|Components|Blackboard|GetValueasVector (AI|GetBlackboard _asfort_player_pawn) _teleportlocationbbkeyname))
        (Variables|Default|SetPawnLocationBeforeTeleport (Transformation|GetActorLocation _asfort_player_pawn))
        (bind _returnvalue (Transform|Actor|TeleportPlayerPawn _asfort_player_pawn (CallFunction|GetTeleportDestination)))
        (if _returnvalue
          ; Execute GC - Teleport Start
          (Ability|ExecuteGameplayCueWithParamsOnOwner 0 (Variables|Default|GetGC%20Converted%20Teleport%20Start) (Ability|GameplayCue|MakeGameplayCueParameters :Location (Variables|Default|GetPawnLocationBeforeTeleport)))
          ; Execute GC - Teleport End
          (Ability|ExecuteGameplayCueWithParamsOnOwner 0 (Variables|Default|GetGC%20Converted%20Teleport%20End) (Ability|GameplayCue|MakeGameplayCueParameters :Location (Transformation|GetActorLocation _asfort_player_pawn)))
          (else
            _))
        (Ability|EndAbility)
        (else
          _)))
    (:CastFailed
      _)))

(event Ability|EventOnEndAbility
  (params (bWasCancelled :type Boolean))
  (|CallParentFunction bWasCancelled) ; This node is disabled and will not be called.\nDrag off pins to build functionality.
)

(fn GetTeleportDestination
  (bind _navfilterclassforpointprojection (Variables|Default|GetNavFilterClassForPointProjection))
  (bind _returnvalue (Ability|GetAvatarActorfromActorInfo))
  (bind _returnvalue_2 (Fortnite|AI|GetNavigationDataforActor _returnvalue))
  (bind _teleportdestination (Variables|Default|GetTeleportDestination))
  (bind (_projectedlocation _returnvalue_3) (AI|Navigation|ProjectPointtoNavigation _teleportdestination _returnvalue_2 _navfilterclassforpointprojection "1000.000000,1000.000000,200.000000"))
  (bind _asfort_pawn (Utilities|Casting|CastToFortPawn _returnvalue)
    (:then
      (Utilities|IsValid _returnvalue_2
        (:Is%20Valid
          (if _returnvalue_3
            (return (Utilities|Operators|Add _projectedlocation (Math|Vector|MakeVector 0.0 0.0 (Components|Capsule|GetScaledCapsuleHalfHeight (Variables|/Script/Engine.Character|GetCapsuleComponent _asfort_pawn)))))
            (else
              _)))
        (:Is%20Not%20Valid
          _))
      (return _teleportdestination))
    (:CastFailed
      _)))
